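# Additive, unlit, two-sided material for an animated glowing corona in the Sonic Spear chest effects.
# Two counter-rotating spikeball flare samples, one inverted and sharpened, combine with a sphere-mask-driven halo gradient, colour tint and HDR remap to form emissive.
# Opacity combines depth and near-camera fades on deferred shading with a vertical UV falloff and Anim scalar.
# Two world position offset variants billboard the mesh via camera-angle rotation, adding scale, location offset and a Juno switch.
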Material /SonicSpearGameplay/Assets/Materials/ChestFX/M_SonicSpear_Cherst_Corona_Combined
graph {
 properties {blendMode=BLEND_Additive, pixelDepthOffsetMode=PDOM_Legacy, shadingModel=MSM_Unlit, twoSided=true, usedShadingModels=Unlit}
 nodes [
  n1 = ObjectOrientation()
  Abs(None=@n1)
  n2 = ObjectOrientation()
  Abs(None=@n2)
  Constant() {r=1}
  n3 = Add(A=TextureCoordinate() {coordinateIndex=1, uTiling=0.25, vTiling=0.25}) {constB=0.375}
  n4 = Multiply(A=Time(), B=ScalarParameter() {defaultValue=0.25, parameterName=RotateSpeed})
  n5 = TextureCoordinate() {coordinateIndex=1, uTiling=0.15, vTiling=0.15}
  n6 = TextureSample(UVs=Rotator(Coordinate=Add(A=@n5) {constB=0.425}, Time=@n4) {speed=-0.72}) {texture=/Game/Effects/Fort_Effects/Textures/Flares/T_Flare_Spikeball}
  n7 = Multiply(
   A=TextureSample(UVs=Rotator(Coordinate=@n3, Time=@n4)) {texture=&T_Flare_Spikeball}.RGB
   B=Power(Base=OneMinus(None=@n6.RGB)) {constExponent=5}
  )
  n8 = ScalarParameter() {defaultValue=7, parameterName=Brightness}
  n9 = Constant() {r=0.5}
  n10 = SphereMask(
   A=TextureCoordinate() {coordinateIndex=1}
   B=@n9
   Radius=@n9
   Hardness=LinearInterpolate(Alpha=ScalarParameter() {defaultValue=0.95, parameterName=Size}) {constA=1, constB=0}
  ) {hardnessPercent=25}
  n11 = Multiply(
   A=VectorParameter() {parameterName=Corona_Colour, defaultValue=<rgba 9 3.7221 0 3>}.RGB
   B=Multiply(
    A=Multiply(A=@n7, B=@n8) {constB=80}
    B=TextureSample(UVs=@n10) {texture=/Game/Effects/Fort_Effects/Textures/Grad/T_Halo_Corona_Gradient}.RGB
   )
  )
  n12 = Clamp(None=@n11)
  MaterialFunctionCall(Emissive=Multiply(
    A=MaterialFunctionCall(LDR Input=@n12, HDR Tint=@n11) {materialFunction=&DeriveHDRfromLDR}.RemappedRange
    B=ScalarParameter() {bUseCustomPrimitiveData=true, defaultValue=1, parameterName=Scaled_Brightness Primitive, primitiveDataIndex=1}
   )) {materialFunction=&MF_FX_EmissiveAdjustments}
  n13 = MaterialFunctionCall(Distance=ScalarParameter() {parameterName=NearCamFade}) {materialFunction=&Near_Camera_Fade}
  n14 = ScalarParameter() {defaultValue=100, parameterName=DepthFade}
  n15 = Multiply(
   A=ScalarParameter() {defaultValue=3, parameterName=Corona_Opacity}
   B=ShadingPathSwitch(Default=Constant() {r=1}, Deferred=DepthFade(Opacity=@n13.Result, FadeDistance=@n14))
  )
  n16 = TextureCoordinate() {coordinateIndex=1}
  n17 = Power(Base=OneMinus(None=ComponentMask(None=@n16) {g=true})) {constExponent=10}
  NamedRerouteDeclaration(None=Multiply(A=Multiply(A=@n15, B=@n17), B=ScalarParameter() {defaultValue=1, parameterName=Anim})) {name=Corona_Opacity, nodeColor=<rgba 1 0.2578 0 1>}
  n18 = Normalize(VectorInput=Subtract(A=ObjectPositionWS(), B=CameraPositionWS()))
  n19 = Arctangent2Fast(Y=ComponentMask(None=@n18) {g=true}, X=ComponentMask(None=@n18) {r=true})
  n20 = MaterialFunctionCall(Multiplier=Constant() {r=2}) {materialFunction=&Pi}
  n21 = RotateAboutAxis(
   NormalizedRotationAxis=@n1
   RotationAngle=Add(A=Divide(A=@n19, B=@n20.Result), B=Constant() {r=0.25}) {constA=0.25, constB=0.25}
   PivotPoint=MaterialFunctionCall() {materialFunction=&ObjectPivotPoint}.Object Pivot Location
   Position=WorldPosition().XYZ
  )
  n22 = MaterialFunctionCall() {materialFunction=&ObjectPivotPoint}
  n23 = Add(A=@n21, B=WorldPosition() {worldPositionShaderOffset=WPT_ExcludeAllShaderOffsets}.XYZ)
  n24 = TransformPosition(None=@n23) {transformSourceType=TRANSFORMPOSSOURCE_World}
  n25 = StaticSwitchParameter(True=Multiply(A=@n24, B=VectorParameter() {parameterName=Scaled_MeshSizeXYZ, defaultValue=<rgba 1 1 1 1>}.RGB) {constB=1.02}, False=@n24) {parameterName=Juno}
  n26 = Add(A=@n21, B=Subtract(A=Add(A=@n22.Object Pivot Location, B=@n25), B=@n23))
  n27 = ObjectPositionWS()
  n28 = WorldPosition()
  n29 = VectorParameter() {parameterName=Scaled_MeshSizeXYZ, defaultValue=<rgba 1 1 1 1>}
  n30 = Add(A=@n26, B=Multiply(A=Subtract(A=@n27, B=@n28.XYZ), B=OneMinus(None=@n29.RGB)))
  n31 = VectorParameter() {parameterName=ObjectLocationOffset, defaultValue=<rgba 0 0 0 1>}
  NamedRerouteDeclaration(None=Add(A=StaticSwitchParameter(True=@n26, False=@n30) {parameterName=Juno}, B=Transform(None=@n31.RGB) {transformSourceType=TRANSFORMSOURCE_Local})) {name=WPO_1, nodeColor=<rgba 0.3125 1 0 1>}
  n32 = Normalize(VectorInput=Subtract(A=ObjectPositionWS(), B=CameraPositionWS()))
  n33 = Arctangent2Fast(Y=ComponentMask(None=@n32) {g=true}, X=ComponentMask(None=@n32) {r=true})
  n34 = MaterialFunctionCall(Multiplier=Constant() {r=2}) {materialFunction=&Pi}
  n35 = RotateAboutAxis(
   NormalizedRotationAxis=@n2
   RotationAngle=Add(A=Divide(A=@n33, B=@n34.Result), B=Constant() {r=0.25}) {constA=0.25, constB=0.25}
   PivotPoint=ActorPositionWS()
   Position=WorldPosition() {worldPositionShaderOffset=WPT_ExcludeAllShaderOffsets}.XYZ
  )
  n36 = ActorPositionWS()
  n37 = Add(A=@n35, B=WorldPosition() {worldPositionShaderOffset=WPT_ExcludeAllShaderOffsets}.XYZ)
  n38 = TransformPosition(None=@n37) {transformSourceType=TRANSFORMPOSSOURCE_World}
  n39 = StaticSwitchParameter(True=Multiply(A=@n38, B=VectorParameter() {parameterName=Scaled_MeshSizeXYZ, defaultValue=<rgba 1 1 1 1>}.RGB) {constB=1.02}, False=@n38) {parameterName=Juno}
  n40 = Add(A=@n35, B=Subtract(A=Add(A=@n36, B=@n39), B=@n37)) {constB=0}
  n41 = ObjectPositionWS()
  n42 = WorldPosition()
  n43 = VectorParameter() {parameterName=Scaled_MeshSizeXYZ, defaultValue=<rgba 1 1 1 1>}
  n44 = Add(A=@n40, B=Multiply(A=Subtract(A=@n41, B=@n42.XYZ), B=OneMinus(None=@n43.RGB)))
  n45 = VectorParameter() {parameterName=ObjectLocationOffset, channelNames {a=VecPos, b="", g="", r=""}, defaultValue=<rgba 0 0 0 1>}
  n46 = Transform(None=@n45.RGB) {transformSourceType=TRANSFORMSOURCE_Local}
  n47 = TransformPosition(None=@n45.RGB) {transformType=TRANSFORMPOSSOURCE_World}
  NamedRerouteDeclaration(None=Add(A=StaticSwitchParameter(True=@n40, False=@n44) {parameterName=Juno}, B=LinearInterpolate(A=@n46, B=@n47, Alpha=@n45.VecPos))) {name=WPO, nodeColor=<rgba 0.3125 1 0 1>}
  NamedRerouteUsage()
  NamedRerouteUsage()
  NamedRerouteUsage()
  ScalarParameter() {defaultValue=1, parameterName=Scaled_MeshSize}
  ScalarParameter() {defaultValue=1, parameterName=Scaled_MeshSize}
 ]
}
